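# Unlit material for the Apprentice Swirl emote's skeletal mesh effect, producing an animated, twisted swirl color.
# Polar UVs, offset oppositely on each UV half, scroll a blurred noise texture that indexes a color curve atlas.
# Vertex red blends this into a hard-edged Fresnel color, then an optional desaturate-and-contrast stage feeds emissive adjustments.
# A camera-depth fade calculation and an edge color parameter are present but feed nothing shown.
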Material /BRCosmetics/Animation/Game/MainPlayer/Emotes/ApprenticeSwirl/FX/ApprenticeSwirlEmote/M_Apprentice_Emote_Swirl_2
graph {
 properties {
  bUsedWithMeshDeformer=true
  bUsedWithSkeletalMesh=true
  pixelDepthOffsetMode=PDOM_Legacy
  previewMesh=/BRCosmetics/Animation/Props/Emote_Apprentice_Swirl/Meshes/Apprentice_Swirl_Emote
  shadingModel=MSM_Unlit
  usedShadingModels=Unlit
 }
 nodes [
  n1 = ComponentMask(None=VectorParameter() {defaultValue=<rgba 0.25 0.27 0 1>}.RGB) {g=true, r=true}
  n2 = Multiply(A=@n1) {constB=-1}
  n3 = TextureCoordinate()
  n4 = Ceil(None=Subtract(A=ComponentMask(None=@n3) {r=true}) {constB=0.5})
  n5 = MaterialFunctionCall!(
   UV=Add(A=TextureCoordinate(), B=LinearInterpolate(A=@n1, B=@n2, Alpha=@n4))
   Twist Center=ScalarParameter() {defaultValue=3, group=Twist, parameterName=TwistAmount}
   Twist Edges
  ) {materialFunction=/Game/Accessories/FORT_Backpacks/Backpack_F_MED_Celestial_Glow/FX/Materials/MF_PolarUV}
  n6 = ScalarParameter() {defaultValue=1, group=Twist, parameterName=TwistScale}
  n7 = Time()
  n8 = MaterialFunctionCall(X=@n7, Y=@n7) {materialFunction=&MakeFloat2}
  n9 = ComponentMask(None=ScalarParameter() {defaultValue=0.5, parameterName=TwistSpeed}) {g=true, r=true}
  n10 = TextureSample!(
   UVs=Add(A=Multiply(A=@n5.Polar UVs, B=@n6), B=Multiply(A=@n8.Result, B=@n9)) {constB=0}
   Tex
   "DDX(UVs)"=@n5.DDX
   "DDY(UVs)"=@n5.DDX
   Apply View MipBias
  ) {mipValueMode=TMVM_Derivative, samplerType=SAMPLERTYPE_Grayscale, texture=&LowResBlurredNoise}
  n11 = CurveAtlasRowParameter(CurveTime=Multiply(A=@n10.R, B=Constant() {r=2})) {
   atlas=&AthenaCurveAtlas01
   curve=/Game/Animation/Game/MainPlayer/Emotes/ApprenticeSwirl/FX/ApprenticeSwirlEmote/Curve_ApprenticeEmote
   defaultValue=322
   parameterName=Param_1
  }
  n12 = VectorParameter() {group=Fresnel, parameterName=FresnelEdgeColor, defaultValue=<rgba 0.2 1 0.2 0>}
  n13 = VectorParameter() {group=Fresnel, parameterName=FresnelCoreColor, defaultValue=<rgba 0.35 0.07 0.2427 1>}
  n14 = ScalarParameter() {defaultValue=4, group=Fresnel, parameterName=FresnelEdge}
  n15 = Ceil(None=MaterialFunctionCall(Exponent=Multiply(A=@n14) {constB=100}) {materialFunction=&MF_Fresnel_FX}.Fresnel)
  n16 = LinearInterpolate(
   A=@n12.RGB
   B=ComponentMask(None=LinearInterpolate(A=@n13.RGB, B=@n12.RGB, Alpha=@n15)) {b=true, g=true, r=true}
   Alpha=ScalarParameter() {defaultValue=1, parameterName=UseFresnel}
  )
  n17 = ScalarParameter() {group=BW, parameterName=Desaturate}
  n18 = LinearInterpolate(A=@n11, B=LinearInterpolate(A=@n16, B=@n16, Alpha=@n17), Alpha=VertexColor().R)
  n19 = Desaturation(None=@n18, Fraction=Constant() {r=1})
  n20 = ScalarParameter() {defaultValue=0.5, group=BW, parameterName=BWContrast}
  MaterialFunctionCall(Emissive=LinearInterpolate(
    A=@n18
    B=MaterialFunctionCall(In=@n19, Contrast=@n20) {materialFunction=&CheapContrast}.Result
    Alpha=@n17
   )) {materialFunction=&MF_FX_EmissiveAdjustments}
  MaterialExpressionSubtract_26 = Subtract(A=CameraPositionWS(), B=WorldPosition().XYZ)
  n21 = Custom!(x=@MaterialExpressionSubtract_26) {
   code="length(x)"
   inputs [{
     inputName=x
     input {expression=&self:MaterialExpressionSubtract_26}
    }]
  }
  n22 = DotProduct(
   A=Normalize(VectorInput=@MaterialExpressionSubtract_26)
   B=Transform(None=Constant3Vector() {constant=<rgba 0 0 -1 0>}) {transformSourceType=TRANSFORMSOURCE_View}
  )
  n23 = Subtract(A=Length(None=Multiply(A=@n21, B=@n22)) {is=class MaterialExpressionLength}, B=ScalarParameter() {parameterName=DistanceOffset})
  n24 = ScalarParameter() {defaultValue=200, parameterName=DistanceLength}
  MaterialExpressionRemap_0 = MaterialXRemap!(None=Clamp(None=Divide(A=@n23, B=@n24)), InputLow, InputHigh, TargetLow, TargetHigh) {targetHighDefault=0.5, targetLowDefault=1}
  VectorParameter() {parameterName=BWEdgeColor, defaultValue=<rgba 1 1 1 0>}
 ]
}
