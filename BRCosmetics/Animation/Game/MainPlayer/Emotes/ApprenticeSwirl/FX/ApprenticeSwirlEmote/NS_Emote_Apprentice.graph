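# Niagara effect for the Apprentice Swirl player emote, attaching colored glows and swirling sparks to two sampled bones or sockets on the character's skeletal mesh.
# Glow bursts two light-emitting sprites that follow those points and cycle color over the system loop.
# Sparks_Gravity001 streams short-lived multicolored sparks from rings with vortex velocity, distance-scaled size and fade; Sparks_Gravity is disabled.
# User.Desaturate blends glow, light and spark colors toward neutral.
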NiagaraSystem /BRCosmetics/Animation/Game/MainPlayer/Emotes/ApprenticeSwirl/FX/ApprenticeSwirlEmote/NS_Emote_Apprentice
dependencies {
 usedDataInterfaces [&NiagaraDataInterfaceColorCurve, &NiagaraDataInterfaceCurve]
 usedDynamicInputs []
 usedModules [
  &SystemState
  &self:SystemSpawnScript.SystemScriptSource.SystemScriptGraph.NiagaraNodeAssignment_0.SetVariables_#1
 ]
}
emitters {
 Glow {
  emitterUpdateScript [
   &EmitterState()
   &SpawnBurst_Instantaneous(Spawn Count=2, Use Spawn Probability=false)
   &SpawnRate(SpawnRate=8, Use Spawn Probability=false) {enabled=false}
  ]
  particleSpawnScript [
   &FN_SetSpawnGroup(Number of Spawn Groups=2)
   &InitializeParticle(
    Color=<rgba 3 3 3 0.8>
    Lifetime=3
    Sprite Rotation Mode=::NewEnumerator2
    Sprite Size Mode=::NewEnumerator3
    Uniform Sprite Size=20
    UsePositionOffset=true
    Write Material Random=true
   ) {module=InitializeParticle001}
   &ShapeLocation(
    MaskSpawnGroup=false
    Offset=&Lerp_Vector(A=<-4.2 -11 3.3>, B=<-4.2 -11 -3.2>, Alpha=&MakeFloatFromInt(Int=$Particles.SpawnGroup))
    Offset Coordinate Space=::Local
    Offset Mode=::NewEnumerator0
    Override Seed=false
    Ring Radius=3
    Rotation Axis=<0.1 1 0>
    Rotation Mode=::NewEnumerator1
    Shape Primitive=::NewEnumerator4
    UseOwnerScale=true
   ) {enabled=false, module=ShapeLocation001}
   &DynamicMaterialParameters(
    Index 0 Param 1=0.5
    Module.Param0WriteEnabled=true
    Module.Param1WriteEnabled=true
    Module.Param2WriteEnabled=true
    Module.Param3WriteEnabled=true
   )
  ]
  particleUpdateScript [
   &ParticleState()
   &SkeletalMeshLocation(Bone / Socket Index=$Particles.SpawnGroup, Bone Sampling Mode=::NewEnumerator4, MaskSpawnGroup=false, Skeletal Mesh=$User.SkeletalMesh) {module=SkeletalMeshLocation001}
   &SkeletalMeshLocation(
    Bone Sampling Mode=::NewEnumerator4
    MaskSpawnGroup=false
    Orientation Sampling=::NewEnumerator1
    Sampled Position Offset=&Lerp_Vector(A=<2 -11 -2.5>, B=<2 -11 2.5>, Alpha=&MakeFloatFromInt(Int=$Particles.SpawnGroup))
    Skeletal Mesh=$User.SkeletalMesh
   ) {enabled=false}
   &SolveForcesAndVelocity(Clamp Velocity=false, Limit Acceleration=false)
   &ScaleColor(
    Curve Index=$System.NormalizedLoopAge
    Linear Color Curve=DataInterface() {propertyValues {
      AlphaCurve {keys [<at 0 1>,<at 1 1>]}
      BlueCurve {keys [<at 0 0>,<at 0.333 0>,<at 0.667 1>,<at 1 0>]}
      GreenCurve {keys [<at 0 0>,<at 0.333 1>,<at 0.667 0>,<at 1 0>]}
      RedCurve {keys [<at 0 1>,<at 0.333 0>,<at 0.667 0>,<at 1 1>]}
     }}
    Scale Mode=::NewEnumerator2
   )
   /SetVariables(Particles.LightColor=&MultiplyLinearColor(
     Linear Color=&LerpLinearColors(StartColor=$Particles.Color, EndColor=<rgba 5 5 5 1>, LerpFactor=$User.Desaturate)
     Scale Factor=&MakeVector4_FromFloat(Float=1.5)
    )) {isSetParameters=true}
   &FN_Light_Attributes(Create Position Attribute=true)
   &ScaleSpriteSize(Scale Factor=&Vector2DFromFloat(Value=&FloatFromCurve(FloatCurve=DataInterface() {propertyValues {Curve {keys [
          {interpMode=RCIM_Cubic, tangentMode=RCTM_User, time=0, value=0}
          {interpMode=RCIM_Cubic, tangentMode=RCTM_User, time=0.4, value=1}
          {interpMode=RCIM_Cubic, tangentMode=RCTM_User, time=1, value=0}
         ]}}}))) {enabled=false}
   &CameraOffset()
   /SetVariables() {isSetParameters=true}
   &Color(Color=&LerpLinearColors(StartColor=$Particles.Color, EndColor=<rgba 2 2 2 1>, LerpFactor=$User.Desaturate))
  ]
  rendererClasses [&NiagaraSpriteRendererProperties, &NiagaraLightRendererProperties]
  renderers [
   {rendererClass=&NiagaraSpriteRendererProperties}
   {rendererClass=&NiagaraLightRendererProperties, rendererIndex=1}
  ]
 }
 Sparks_Gravity {
  bEnabled=false
  emitterUpdateScript [
   &EmitterState() {module=EmitterState001}
   &SpawnRate(SpawnGroup=1, SpawnRate=200, Use Spawn Probability=false) {module=SpawnRate001}
   /SetVariables(Emitter.VelocityPosition:Vector3f=&RandomRangeVector(Minimum=<-20 -20 -20>, Maximum=<20 20 20>, Evaluation Type=::NewEnumerator1, Random Seed=$Emitter.LoopCount, Fixed Random Seed=true)) {enabled=false, isSetParameters=true, module=SetVariables_#2}
  ]
  particleSpawnScript [
   &FN_SetSpawnGroup(Number of Spawn Groups=2)
   /Niagara/Modules/Spawn/Initialization/InitializeParticle(
    Color=&LerpLinearColors(
     StartColor=&MultiplyLinearColor(
      Linear Color=&ColorFromCurve(
       ColorCurve=DataInterface() {propertyValues {
         AlphaCurve {keys [<at 0 1 RCIM_Cubic>]}
         BlueCurve {keys [<at 0 0.3054 RCIM_Cubic>,<at 0.2352 1>,<at 0.4678 0.8873>,<at 0.8219 0 RCIM_Cubic>]}
         GreenCurve {keys [<at 0 1 RCIM_Cubic>,<at 0.2352 0.0363>,<at 0.4678 0>,<at 0.8219 0.3634 RCIM_Cubic>]}
         RedCurve {keys [<at 0 0.0645 RCIM_Cubic>,<at 0.2352 0>,<at 0.4678 1>,<at 0.8219 1 RCIM_Cubic>]}
        }}
       CurveIndex=&RandomRangeFloat(Minimum=0, Maximum=1)
      )
      Scale Factor=<0.3 0.3 0.3 10>
     )
     EndColor=<rgba 0.2 0.2 0.2 1>
     LerpFactor=$User.Desaturate
    )
    Lifetime=&Multiply_Float(A=&UniformRangedFloat(Minimum=0.4, Maximum=0.6), B=0.5)
    Module.Write Color=true
    Module.Write Lifetime=true
    Module.Write Mass=true
    Module.Write Material Random=true
    Module.Write Position=false
    Module.Write Scale=false
    Module.Write SpriteRotation=false
    Module.Write SpriteSize=true
    Position=&SimulationPosition()
    Sprite Size=&RandomRangeVector2D(Minimum=<0.2 1>, Maximum=<0.6 0.6>)
   )
   &SkeletalMeshLocation(
    Bone Sampling Mode=::NewEnumerator4
    Interpolate Mesh Sampling=false
    MaskSpawnGroup=false
    Orientation Sampling=::NewEnumerator1
    Position Sampling=::NewEnumerator2
    Sampled Position Offset=&Lerp_Vector(A=<2 -10.5 -4>, B=<2 -10.5 4>, Alpha=&MakeFloatFromInt(Int=$Particles.SpawnGroup))
    Skeletal Mesh=$User.SkeletalMesh
   )
   &ShapeLocation(
    Additional Quaternion Rotation=true
    Additional Yaw / Pitch / Roll=<0.01 0 0.25>
    MaskSpawnGroup=false
    Offset Coordinate Space=::Local
    Offset Mode=::NewEnumerator0
    Override Seed=false
    Ring Radius=3
    Rotation Mode=::NewEnumerator3
    Rotation Quaternion=$StackContext.SkeletalMeshLocation.SampledRotation
    Shape Origin=$Particles.SkeletalMeshLocation.SampledPosition
    Shape Primitive=::NewEnumerator4
    UseOwnerScale=true
   ) {module=ShapeLocation001}
   &ShapeLocation(
    Handle Radius=1
    Large Radius=2
    MaskSpawnGroup=false
    Offset Coordinate Space=::Local
    Offset Mode=::NewEnumerator0
    Override Seed=false
    Rotation Axis=<0 1 0>
    Rotation Mode=::NewEnumerator1
    Shape Origin=$Particles.SkeletalMeshLocation.SampledPosition
    Shape Primitive=::NewEnumerator3
    UseOwnerScale=true
   ) {enabled=false}
   &AddVelocity(
    Module.Use Falloff Distance=false
    Override Seed=false
    Velocity Mode=::NewEnumerator1
    Velocity Origin=$Particles.SkeletalMeshLocation.SampledPosition
    Velocity Speed=0
   ) {enabled=false}
   /Niagara/Modules/Update/Velocity/VortexVelocity(
    Influence Falloff Radius=400
    Module.Use Influence Falloff=false
    Velocity Amount:NiagaraFloat=&RandomRangeFloat(Minimum=-5, Maximum=-15)
    Vortex Axis:Vector3f=<0 1 0>
    Vortex Origin:NiagaraPosition=$Particles.SkeletalMeshLocation.SampledPosition
   )
   &ApplyInitialForces()
  ]
  particleUpdateScript [
   &ParticleState()
   &ScaleSpriteSize(
    Uniform Curve Scale=1.3
    Uniform Curve Sprite Scale=DataInterface() {propertyValues {Curve {keys [{tangentMode=RCTM_User, time=1, value=1}]}}}
   ) {module=ScaleSpriteSize002}
   &VortexVelocity(
    Influence Falloff Radius=400
    Module.Use Influence Falloff=false
    Velocity Amount=-800
    Vortex Axis:Vector3f=<0 1 0>
    Vortex Origin:NiagaraPosition=$Particles.SkeletalMeshLocation.SampledPosition
   ) {module=VortexVelocity001}
   &Drag(Drag=0, Use Linear Drag=true, Use Rotational Drag=false)
   &GravityForce(Gravity=<0 0 -500>) {enabled=false}
   &SolveForcesAndVelocity(Clamp Velocity=false, Limit Acceleration=false)
   &ScaleColor(
    Scale Alpha=&FloatFromCurve(FloatCurve=DataInterface() {propertyValues {Curve {keys [<at 0 1>,<at 1 0>]}}})
    Scale RGB=&VectorFromFloat(Value=&FloatFromCurve(FloatCurve=DataInterface() {propertyValues {Curve {keys [<at 0 50>,<at 0.3 35>,<at 1 0>]}}}))
    ScaleA=true
    ScaleRGB=true
   )
   &ScaleSpriteSize(Scale Factor=&Vector2DFromFloat(Value=&FloatFromCurve(
      FloatCurve=DataInterface() {propertyValues {Curve {keys [
          <at 0 0.15 RCIM_Cubic>
          {interpMode=RCIM_Cubic, tangentMode=RCTM_User, time=0.3, value=1.5}
          {interpMode=RCIM_Cubic, tangentMode=RCTM_User, time=1, value=0}
         ]}}}
      Scale Curve=&MakeFloatFromVector2d(Vector 2D=&Vector2DFromFloat(Value=&FN_LOD_Distance_Remap(Start Distance=300, End Distance=1000, Minimum Output=1.1, Maximum Output=2)))
     )))
   &ScaleColor(Scale Alpha=$System.DistanceFloat, ScaleA=true, ScaleRGB=true) {module=ScaleColor001}
   &ScaleSpriteSize(
    Uniform Curve Scale=0.5
    Uniform Curve Sprite Scale=DataInterface() {propertyValues {Curve {keys [{tangentMode=RCTM_User, time=1, value=1}]}}}
   ) {enabled=false, module=ScaleSpriteSize001}
  ]
  rendererClasses [&NiagaraSpriteRendererProperties]
  renderers [{rendererClass=&NiagaraSpriteRendererProperties}]
 }
 Sparks_Gravity001 {
  emitterUpdateScript [
   &EmitterState() {module=EmitterState001}
   &SpawnRate(SpawnGroup=1, SpawnRate=200, Use Spawn Probability=false) {module=SpawnRate001}
   /SetVariables(Emitter.VelocityPosition:Vector3f=&RandomRangeVector(Minimum=<-20 -20 -20>, Maximum=<20 20 20>, Evaluation Type=::NewEnumerator1, Random Seed=$Emitter.LoopCount, Fixed Random Seed=true)) {enabled=false, isSetParameters=true, module=SetVariables_#2}
  ]
  particleSpawnScript [
   &FN_SetSpawnGroup(Number of Spawn Groups=2)
   &InitializeParticle(
    Color=&LerpLinearColors(
     StartColor=&MultiplyLinearColor(
      Linear Color=&ColorFromCurve(
       ColorCurve=DataInterface() {propertyValues {
         AlphaCurve {keys [<at 0 1 RCIM_Cubic>]}
         BlueCurve {keys [<at 0 0.3054 RCIM_Cubic>,<at 0.2352 1>,<at 0.4678 0.8873>,<at 0.8219 0 RCIM_Cubic>]}
         GreenCurve {keys [<at 0 1 RCIM_Cubic>,<at 0.2352 0.0363>,<at 0.4678 0>,<at 0.8219 0.3634 RCIM_Cubic>]}
         RedCurve {keys [<at 0 0.0645 RCIM_Cubic>,<at 0.2352 0>,<at 0.4678 1>,<at 0.8219 1 RCIM_Cubic>]}
        }}
       CurveIndex=&RandomRangeFloat(Minimum=0, Maximum=1)
      )
      Scale Factor=<0.3 0.3 0.3 10>
     )
     EndColor=<rgba 0.2 0.2 0.2 1>
     LerpFactor=$User.Desaturate
    )
    Lifetime=&Multiply_Float(A=&UniformRangedFloat(Minimum=0.4, Maximum=0.6), B=0.5)
    Module.Write Color=true
    Module.Write Lifetime=true
    Module.Write Mass=true
    Module.Write Material Random=true
    Module.Write Position=false
    Module.Write Scale=false
    Module.Write SpriteRotation=false
    Module.Write SpriteSize=true
    Position=&SimulationPosition()
    Sprite Size=&RandomRangeVector2D(Minimum=<0.2 1>, Maximum=<0.6 0.6>)
   )
   &SkeletalMeshLocation(Bone / Socket Index=$Particles.SpawnGroup, Bone Sampling Mode=::NewEnumerator4, MaskSpawnGroup=false, Skeletal Mesh=$User.SkeletalMesh) {module=SkeletalMeshLocation001}
   &ShapeLocation(
    MaskSpawnGroup=false
    Offset=&Lerp_Vector(A=<0 0 0>, B=<0 0 0>, Alpha=&MakeFloatFromInt(Int=$Particles.SpawnGroup))
    Offset Coordinate Space=::Local
    Offset Mode=::NewEnumerator0
    Override Seed=false
    Ring Radius=3
    Rotation Axis=<0.1 1 0>
    Rotation Mode=::NewEnumerator1
    Shape Primitive=::NewEnumerator4
    UseOwnerScale=true
   ) {module=ShapeLocation001}
   /SetVariables(Particles.Position=&AddVectorToPosition(Position=$Particles.ShapeLocation001.ShapePosition, Vector=<0 0 0>)) {enabled=false, isSetParameters=true}
   &ShapeLocation(
    Handle Radius=1
    Large Radius=2
    MaskSpawnGroup=false
    Offset Coordinate Space=::Local
    Offset Mode=::NewEnumerator0
    Override Seed=false
    Rotation Axis=<0 1 0>
    Rotation Mode=::NewEnumerator1
    Shape Origin=$Particles.SkeletalMeshLocation.SampledPosition
    Shape Primitive=::NewEnumerator3
    UseOwnerScale=true
   ) {enabled=false}
   &VortexVelocity(
    Influence Falloff Radius=400
    Module.Use Influence Falloff=false
    Velocity Amount:NiagaraFloat=&RandomRangeFloat(Minimum=-5, Maximum=-15)
    Vortex Axis:Vector3f=<0 2 0>
    Vortex Origin:NiagaraPosition=$Particles.SkeletalMeshLocation001.SampledPosition
   )
   &AddVelocity(Velocity=<0 -5 0>)
   &ApplyInitialForces()
  ]
  particleUpdateScript [
   &ParticleState()
   &ScaleSpriteSize(
    Uniform Curve Scale=1.3
    Uniform Curve Sprite Scale=DataInterface() {propertyValues {Curve {keys [{tangentMode=RCTM_User, time=1, value=1}]}}}
   ) {module=ScaleSpriteSize002}
   &VortexVelocity(Influence Falloff Radius=400, Module.Use Influence Falloff=false, Velocity Amount=-800, Vortex Axis:Vector3f=<0 1 0>) {enabled=false, module=VortexVelocity001}
   &Drag(Drag=0, Use Linear Drag=true, Use Rotational Drag=false)
   &GravityForce(Gravity=<0 0 -500>) {enabled=false}
   &SolveForcesAndVelocity(Clamp Velocity=false, Limit Acceleration=false)
   &ScaleColor(
    Scale Alpha=&FloatFromCurve(FloatCurve=DataInterface() {propertyValues {Curve {keys [<at 0 1>,<at 1 0>]}}})
    Scale RGB=&VectorFromFloat(Value=&FloatFromCurve(FloatCurve=DataInterface() {propertyValues {Curve {keys [<at 0 50>,<at 0.3 35>,<at 1 0>]}}}))
    ScaleA=true
    ScaleRGB=true
   )
   &ScaleSpriteSize(Scale Factor=&Vector2DFromFloat(Value=&FloatFromCurve(
      FloatCurve=DataInterface() {propertyValues {Curve {keys [
          <at 0 0.15 RCIM_Cubic>
          {interpMode=RCIM_Cubic, tangentMode=RCTM_User, time=0.3, value=1.5}
          {interpMode=RCIM_Cubic, tangentMode=RCTM_User, time=1, value=0}
         ]}}}
      Scale Curve=&MakeFloatFromVector2d(Vector 2D=&Vector2DFromFloat(Value=&FN_LOD_Distance_Remap(Start Distance=300, End Distance=1000, Minimum Output=1.1, Maximum Output=2)))
     )))
   &ScaleColor(Scale Alpha=$System.DistanceFloat, ScaleA=true, ScaleRGB=true) {module=ScaleColor001}
   &ScaleSpriteSize(
    Uniform Curve Scale=0.5
    Uniform Curve Sprite Scale=DataInterface() {propertyValues {Curve {keys [{tangentMode=RCTM_User, time=1, value=1}]}}}
   ) {enabled=false, module=ScaleSpriteSize001}
  ]
  rendererClasses [&NiagaraSpriteRendererProperties]
  renderers [{rendererClass=&NiagaraSpriteRendererProperties}]
 }
}
summary {
 systemName=NS_Emote_Apprentice
 emitters [
  {bEnabled=false, emitterName=Sparks_Gravity, rendererClasses [&NiagaraSpriteRendererProperties]}
  {emitterName=Glow, rendererClasses [&NiagaraSpriteRendererProperties, &NiagaraLightRendererProperties]}
  {emitterName=Sparks_Gravity001, rendererClasses [&NiagaraSpriteRendererProperties]}
 ]
 userVariables [
  User.Desaturate {type=NiagaraFloat, defaultValue=0, description=""}
  User.FrontendSpawn {type=NiagaraBool}
  User.RibbonIDOffset {type=NiagaraInt32}
  User.SkeletalMesh {
   type=NiagaraDataInterfaceSkeletalMesh
   defaultValue {
    struct=NiagaraExt_VariableValue_DataInterface
    value {dataInterface=&self:NiagaraDataInterfaceSkeletalMesh_1}
   }
  }
 ]
}
